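; Parent Blueprint for the Commerce Restaurant Display B prop, with configurable materials and colors for its bun box and soup bun box meshes.
; The construction script calls a helper that validates each mesh component, sets slot 0 to the override material, then applies the tint color.
; It bumps the exposed asset provider version and registers both override materials as allowed asset references.
; BeginPlay, overlap and Tick events are empty.
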(event EventBeginPlay)

(event Collision|EventActorBeginOverlap
  (params (OtherActor :type Actor%20Object%20Reference)))

(event EventTick
  (params (DeltaSeconds :type Float%20%28single-precision%29)))

(fn AssignMaterialToMeshComponent
  (params (MaterialSlot :type Integer) (NewMaterial :type Material%20Interface%20Object%20Reference) (StaticMesh :type Static%20Mesh%20Component%20Object%20Reference) (TintColor :type Linear%20Color%20Structure) (UseTintColorFromOverride :type Boolean))
  (Utilities|IsValid StaticMesh
    (:Is%20Valid
      (Utilities|FlowControl|Sequence
        (:then_0
          (Rendering|Material|SetMaterial StaticMesh MaterialSlot NewMaterial))
        (:then_1
          (if UseTintColorFromOverride
            (Default|SetColoronSecondaryMesh TintColor StaticMesh)))))))

(fn BuildExposedAssetProviderVersion(Implementation)
  (params (Version :type Fort%20Exposed%20Asset%20Provider%20Version%20Structure%20%28by%20ref%29))
  (|CallParentFunction Version)
  (ExposedAssetProvider|AddVersion Version 2)
  (return))

(fn CollectAssetReferencestoAllow(Implementation)
  (params (Collector :type Fort%20Exposed%20Asset%20Reference%20Collector%20Structure%20%28by%20ref%29))
  (|CallParentFunction Collector)
  (ExposedAssetProvider|CollectObjectPropertyAssetReferences Collector 0 (Variables|Default|GetBunBoxMaterialOverride))
  (ExposedAssetProvider|CollectObjectPropertyAssetReferences Collector 0 (Variables|Default|GetSoupBunBoxMaterialOverride))
  (return))

(fn ConstructionScript
  (|CallParentFunction)
  (CallFunction|AssignMaterialtoMeshComponent 0 (Variables|Default|GetBunBoxMaterialOverride) (Variables|Default|GetBunBoxInstances) (Variables|Default|GetBunBoxColor) true)
  (CallFunction|AssignMaterialtoMeshComponent 0 (Variables|Default|GetSoupBunBoxMaterialOverride) (Variables|Default|GetSoupBunBoxes) (Variables|Default|GetSoupBunBoxColor) true))
